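# Animation Blueprint for the GalliumReality pickaxe that spins its wheel bones, much faster during jumps or melee animations.
# EventGraph eases a jump-or-melee value to set spin speed, accumulates it into a roll rotator, and converts the front-end flag to a float.
# AnimGraph runs RigidBody weighted by that flag, then additively rotates both upper wheel bones in bone space.
# Lower-wheel nodes are disconnected from output; spin_fnc is uncalled.
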AnimBlueprint /BRCosmetics/Weapons/FORT_Melee/Pickaxe_GalliumReality/Meshes/GalliumReality_Axe_AnimBP
target_skeleton=&/GalliumReality_Axe_Skeleton
graphs [
 AnimGraph {nodes [
   n1 = Get wheel_rotate() -> (wheel_rotate:Rotator)
   n2 = "Transform (Modify) Bone - Bone: wheel_lower_l_base"(
    ComponentPose=Reroute Node(InputPin:Component Space Pose Link Structure) -> (OutputPin:Component Space Pose Link Structure)
    Alpha=1
    Translation="0,0,0"
    Rotation=@n1
    Scale="1,1,1"
   ) {
    rotationMode=BMM_Additive
    rotationSpace=BCS_BoneSpace
    translationMode=BMM_Additive
    translationSpace=BCS_BoneSpace
    alphaBoolBlend {blendOption=Linear}
    boneToModify {boneName=wheel_lower_l_base}
   }
   "Transform (Modify) Bone - Bone: wheel_lower_r_base"(ComponentPose=@n2, Alpha=1, Translation="0,0,0", Rotation=@n1, Scale="1,1,1") {
    rotationMode=BMM_Additive
    rotationSpace=BCS_BoneSpace
    translationMode=BMM_Additive
    translationSpace=BCS_BoneSpace
    alphaBoolBlend {blendOption=Linear}
    boneToModify {boneName=wheel_lower_r_base}
   }
   n3 = Input Pose()
   n4 = Get wheel_rotate()
   n5 = "Transform (Modify) Bone - Bone: wheel_upper_l_base"(
    ComponentPose=RigidBody(ComponentPose=Local To Component(LocalPose=@n3), Alpha=Get IsinFE() -> (IsinFE:double)) {alphaBoolBlend {blendOption=Linear}}
    Alpha=1
    Translation="0,0,0"
    Rotation=@n4
    Scale="1,1,1"
   ) {
    rotationMode=BMM_Additive
    rotationSpace=BCS_BoneSpace
    translationMode=BMM_Additive
    translationSpace=BCS_BoneSpace
    alphaBoolBlend {blendOption=Linear}
    boneToModify {boneName=wheel_upper_l_base}
   }
   n6 = "Transform (Modify) Bone - Bone: wheel_upper_r_base"(ComponentPose=@n5, Alpha=1, Translation="0,0,0", Rotation=@n4, Scale="1,1,1") {
    rotationMode=BMM_Additive
    rotationSpace=BCS_BoneSpace
    translationMode=BMM_Additive
    translationSpace=BCS_BoneSpace
    alphaBoolBlend {blendOption=Linear}
    boneToModify {boneName=wheel_upper_r_base}
   }
   Output Pose(Result=Component To Local(ComponentPose=@n6))
  ]}
 EventGraph {nodes [
   n1 = Event BlueprintUpdateAnimation()
   > n2 = Sequence!(execute) -> (then_0, then_1, then_2:Exec)
   n3 = OR Boolean(A=Get bIsJumping(), B=Get bIsPlayingMeleeAnim())
   n4 = Set Is Playing Melee Anim(
    execute=@n2.then_1
    Is Playing Melee Anim:double=FInterpTo(
     Current=Get Is Playing Melee Anim() -> (Is Playing Melee Anim:double)
     Target="To Float (Boolean)"(InBool=@n3)
     DeltaTime=@n1.DeltaTimeX
     InterpSpeed=1.5
    )
   ) -> (Output_Get:double)
   n5 = float * float(A=1, B=Lerp(A=2, B=1200, Alpha=@n4))
   n6 = Branch(execute=@n2.then_0, Condition="float >= float"(A=Get wheel_spin() -> (wheel_spin:double), B=@n5))
   n7 = float + float(A=Get wheel_spin(), B=float * float(A=@n5, B=@n1.DeltaTimeX) {is=title K2Node_PromotableOperator}) {is=title K2Node_CommutativeAssociativeBinaryOperator}
   > n8 = Set wheel_spin(wheel_spin:double=@n7) -> (Output_Get:double)
   Set wheel_rotate(execute=[@n9.then, @n8.then], wheel_rotate:Rotator=MakeRotator(Roll=@n7)) -> (Output_Get:Rotator)
   n9 = Set wheel_spin(execute=@n6.else, wheel_spin=@n7)
   Set IsinFE(execute=@n2.then_2, IsinFE:double="To Float (Boolean)"(InBool=Get bIsInFrontEnd())) -> (Output_Get:double)
  ]}
 spin_fnc {nodes [
   n1 = spin_fnc!() -> (then, Speed:double=~, Init Val:double=~) {is=title K2Node_FunctionEntry}
   n2 = float * float(A=@n1, B=float - float(A=GetRealTimeSeconds())) {is=title K2Node_CommutativeAssociativeBinaryOperator}
   > Return Node!(execute, output:double=NormalizeAxis!(Angle:float=float + float(A=@n1.Init Val, B=@n2)) -> (ReturnValue:float))
  ]}
]
variables [
 Delta Time X {type=number}
 Game Time in Seconds {type=number}
 Gravity Override Params {type=GravityOverrideParamsStruct}
 GravityOverride {type=Vector}
 Is Playing Melee Anim {type=number}
 IsinFE {type=number}
 wheel_rotate {type=Rotator, default=<rot 0 0 726.6129>}
 wheel_spin {type=number, default=726.5796}
]
