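# Physics asset for the GalliumReality two-handed pickaxe, giving its four wheels bodies that can spin on the weapon.
# One capsule covers the handle, and each upper and lower, left and right wheel has a tiny base capsule plus a larger dyn capsule.
# Each base joins the handle with free swing and twist.
# Each dyn joins its base with free twist and near-zero swing limits, so it turns about a single axis like an axle.
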PhysicsAsset /BRCosmetics/Weapons/FORT_Melee/Pickaxe_GalliumReality/Meshes/GalliumReality_Axe_PhysicsAsset
bodies {
 Melee_TwoHanded_Handle [{shapeType=Capsule, length=66.0977, radius=17.5765, center=<3.9717 -8.45 -2.9048>, rotation=<rot 0.0286 -90 0.0007>}]
 wheel_lower_l_base [{shapeType=Capsule, length=0.505, radius=0.505, shapeName=wheel_lower_l_base_capsule, rotation=<rot 0 -90 0>}]
 wheel_lower_l_dyn [{shapeType=Capsule, length=8, radius=5, shapeName=wheel_lower_l_dyn_capsule, center=<0 0 0.2>}]
 wheel_lower_r_base [{shapeType=Capsule, length=0.505, radius=0.505, shapeName=wheel_lower_r_base_capsule, rotation=<rot 0 -90 0>}]
 wheel_lower_r_dyn [{shapeType=Capsule, length=8, radius=5, shapeName=wheel_lower_r_dyn_capsule, center=<0 0 0.4>}]
 wheel_upper_l_base [{shapeType=Capsule, length=0.505, radius=0.505, shapeName=wheel_upper_l_base_capsule, rotation=<rot 0 -90 0>}]
 wheel_upper_l_dyn [{shapeType=Capsule, length=8, radius=5, shapeName=wheel_upper_l_dyn_capsule, center=<0 0 0.5>}]
 wheel_upper_r_base [{shapeType=Capsule, length=0.505, radius=0.505, shapeName=wheel_upper_r_base_capsule, rotation=<rot 0 -90 0>}]
 wheel_upper_r_dyn [{shapeType=Capsule, length=8, radius=5, shapeName=wheel_upper_r_dyn_capsule, center=<0 0 0.3>}]
}
constraints [
 {
  bone1Name=wheel_upper_r_dyn
  bone2Name=wheel_upper_r_base
  swing1LimitDegrees=0.1
  swing2LimitDegrees=0.1
  twistLimitDegrees=45
  twistMotion=Free
 }
 {
  bone1Name=wheel_lower_l_base
  bone2Name=Melee_TwoHanded_Handle
  swing1LimitDegrees=45
  swing1Motion=Free
  swing2LimitDegrees=45
  swing2Motion=Free
  twistLimitDegrees=45
  twistMotion=Free
 }
 {
  bone1Name=wheel_lower_l_dyn
  bone2Name=wheel_lower_l_base
  swing1LimitDegrees=0.1
  swing2LimitDegrees=0.1
  twistLimitDegrees=45
  twistMotion=Free
 }
 {
  bone1Name=wheel_lower_r_base
  bone2Name=Melee_TwoHanded_Handle
  swing1LimitDegrees=45
  swing1Motion=Free
  swing2LimitDegrees=45
  swing2Motion=Free
  twistLimitDegrees=45
  twistMotion=Free
 }
 {
  bone1Name=wheel_lower_r_dyn
  bone2Name=wheel_lower_r_base
  swing1LimitDegrees=0.1
  swing2LimitDegrees=0.1
  twistLimitDegrees=45
  twistMotion=Free
 }
 {
  bone1Name=wheel_upper_l_base
  bone2Name=Melee_TwoHanded_Handle
  swing1LimitDegrees=45
  swing1Motion=Free
  swing2LimitDegrees=45
  swing2Motion=Free
  twistLimitDegrees=45
  twistMotion=Free
 }
 {
  bone1Name=wheel_upper_l_dyn
  bone2Name=wheel_upper_l_base
  swing1LimitDegrees=0.1
  swing2LimitDegrees=0.1
  twistLimitDegrees=45
  twistMotion=Free
 }
 {
  bone1Name=wheel_upper_r_base
  bone2Name=Melee_TwoHanded_Handle
  swing1LimitDegrees=45
  swing1Motion=Free
  swing2LimitDegrees=45
  swing2Motion=Free
  twistLimitDegrees=45
  twistMotion=Free
 }
]
